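# Conversation database for the NPC MarkerDeer's "Ask about..." dialogue, entered through its Dialogue entry tag.
# A greeting speech node branches into Becoming a Pilot and The Mission topics, a Back choice, and a link node.
# The link node pulls in the remote MarkerDeer TrainingCompleted entry by tag, outside this graph.
# Both topic replies and the link node share one ReturnToStart BACK node.
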ConversationDatabase /JethroLager/NPCs/Tandem/MarkerDeer/Jethro_Conversation_MarkerDeer_Dialogue
entry_points [{tag=Jethro.Conversation.Entry.MarkerDeer.Dialogue, destinations [#1]}]
nodes [
 ConversationEntryPointNode_0 {
  id=#1
  class=ConversationEntryPointNode
  outputs [#2]
  properties {entryTag {tagName=Jethro.Conversation.Entry.MarkerDeer.Dialogue}}
 }
 Speech_C_0 {
  id=#2
  class=Speech_C
  outputs [#3, #4, #5, #6]
  properties {nodeName="", softTaskIcon=&T_UI_NPC_question, generalConfig {defaultMessage=How're you holding up?}}
  sub_nodes [ConversationChoiceNode_5 {class=ConversationChoiceNode, properties {defaultChoiceDisplayText=Ask about...}}]
 }
 ConversationLinkNode_0 {
  id=#6
  class=ConversationLinkNode
  outputs [#7]
  properties {remoteEntryTag {tagName=Conversation.Entry.MarkerDeer.TrainingCompleted}}
 }
 ReturnToStart_C_23 {
  id=#7
  class=ReturnToStart_C
  properties {softTaskIcon=&T_UI_NPC_Back}
  sub_nodes [ConversationChoiceNode_32 {class=ConversationChoiceNode, properties {defaultChoiceDisplayText=BACK}}]
 }
 Speech_C_2 {
  id=#5
  class=Speech_C
  outputs [#7]
  properties {
   nodeName=""
   softTaskIcon=&T_UI_NPC_question
   generalConfig {defaultMessage=They tell me to fly and I fly. That's enough for me.}
  }
  sub_nodes [ConversationChoiceNode_0 {class=ConversationChoiceNode, properties {defaultChoiceDisplayText=The Mission}}]
 }
 Back_C_0 {
  id=#4
  class=Back_C
  sub_nodes [ConversationChoiceNode_7 {class=ConversationChoiceNode, properties {defaultChoiceDisplayText=BACK}}]
 }
 Speech_C_7 {
  id=#3
  class=Speech_C
  outputs [#7]
  properties {
   nodeName=""
   softTaskIcon=&T_UI_NPC_question
   generalConfig {defaultMessage=You never know what you're capable of until you give it a shot.}
  }
  sub_nodes [ConversationChoiceNode_1 {class=ConversationChoiceNode, properties {defaultChoiceDisplayText=Becoming a Pilot}}]
 }
]
